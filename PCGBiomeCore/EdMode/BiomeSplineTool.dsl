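; BiomeSplineTool is an empty graph in the EdMode folder of PCGBiomeCore, named as a spline tool for biome editing.
; It contains no nodes, connections or parameters, so it defines no behavior of its own.
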(graph BiomeSplineTool)
