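; An empty graph asset named BiomeVolumeTool in the EdMode folder of PCGBiomeCore.
; Its name and location suggest an editor-mode tool for biome volumes, but the graph contains no nodes, variables or connections.
; It defines no behaviour of its own in this file.
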(graph BiomeVolumeTool)
